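# Flipbook master material in the Joels_Bar environment set, built on material attributes around the SubUVTextures function.
# A base with black color and low roughness feeds that function, with world position offset from Fortnite_HitEffects for hit-and-shrink deformation.
# A static switch can darken-blend base color and emissive against a tint.
# Another sets opacity from a UV2 gradient, Fresnel-blended values and depth fade, which the mobile path skips.
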Material /Game/Environments/Sets/Joels_Bar/Material/M_SimpleFlipbook
graph {
 properties {
  bUseMaterialAttributes=true
  bUsedWithMeshDeformer=true
  bUsedWithNanite=true
  bUsedWithSkeletalMesh=true
  pixelDepthOffsetMode=PDOM_Legacy
  usedShadingModels=Default Lit
 }
 nodes [
  ScalarParameter() {group=z_code, parameterName=Code_HitAnimation}
  n1 = Power(
   Base=ComponentMask(None=TextureCoordinate() {coordinateIndex=2}) {g=true}
   Exp=ScalarParameter() {defaultValue=2, group=Opacity, parameterName=AlphaGradient-Exponent}
  ) {constExponent=2.2}
  n2 = ScalarParameter() {group=Opacity, parameterName=Parallel-Value}
  n3 = ScalarParameter() {defaultValue=0.8, group=Opacity, parameterName=Perpendicular-Value}
  n4 = MaterialFunctionCall(
   Brightness=ScalarParameter() {defaultValue=1.5, group=Opacity, parameterName=Fresnel-Brightness}
   Exponent=ScalarParameter() {defaultValue=2, group=Opacity, parameterName=Fresnel-Exponent}
  ) {materialFunction=&MF_Fresnel_FX}
  n5 = Multiply(
   A=Multiply(A=Clamp(None=@n1), B=LinearInterpolate(A=@n2, B=@n3, Alpha=@n4.Fresnel))
   B=ScalarParameter() {defaultValue=1, group=Opacity, parameterName=Opacity}
  )
  n6 = Constant() {r=1}
  n7 = ShadingPathSwitch(Default=DepthFade(Opacity=@n6, FadeDistance=ScalarParameter() {defaultValue=15, parameterName=DepthFadeDistance}), Mobile=@n6) {desc=Migrated from FeatureLevelSwitch}
  n8 = VectorParameter() {group=z_code, parameterName=WSDeformationCenter}
  n9 = ScalarParameter() {group=z_code, parameterName=Code_DestructionAmount}
  n10 = StaticBoolParameter() {group=z_code, parameterName=Code_UseHitEffects}
  n11 = ScalarParameter() {defaultValue=1, group=z_code, parameterName=WSDamageRadius}
  n12 = ScalarParameter() {group=z_code, parameterName=DistanceCullFadeAnimationOffset}
  n13 = MaterialFunctionCall!(
   Alpha=Constant() {r=1}
   Material=MakeMaterialAttributes!(
    BaseColor=Constant()
    Metallic
    Specular=Constant() {r=0.5}
    Roughness=Constant() {r=0.2}
    Anisotropy
    EmissiveColor
    Opacity
    OpacityMask
    Normal
    Tangent
    WorldPositionOffset=MaterialFunctionCall!(
     WSDeformationCenter=@n8.RGB
     Code_DestructionAmount=@n9
     Code_UseUvForDamage
     Code_UseHitEffects=@n10
     Code_Emissive
     WSDamageRadius=@n11
     CrackDamageShader
     Diffuse
     ScalarModDamageCracks
     ScalarOffsetforDamageCracks
     ExposeLerpFunctionForInvertingTheHitDistanceCalc
     InvertDamageFalloff
     CrackWorldPosition
     WorldPosition
     DistanceCullFadeAnimationOffset=@n12
    ) {materialFunction=/Game/Packages/jl_tests/Fortnite_HitEffects}.Hit And Shrink WPO
    SubsurfaceColor
    ClearCoat
    ClearCoatRoughness
    AmbientOcclusion
    Refraction
    CustomizedUVs_0
    CustomizedUVs_1
    CustomizedUVs_2
    CustomizedUVs_3
    CustomizedUVs_4
    CustomizedUVs_5
    CustomizedUVs_6
    CustomizedUVs_7
    PixelDepthOffset
    ShadingModel
    Displacement
   )
  ) {materialFunction=/Game/Weapons/WeaponSkins/MaterialFunction/SubUVTextures}
  n14 = GetMaterialAttributes(None=@n13.Result) {attributeGetTypes [#1, #2]}
  n15 = VectorParameter() {parameterName=Darken Blend, defaultValue=<rgba 1 1 1 1>}
  n16 = StaticSwitchParameter(
   True=SetMaterialAttributes!(
    MaterialAttributes=@n14.MaterialAttributes
    Emissive Color=MaterialFunctionCall!(Blend=@n14.Emissive Color, Base=@n15.RGB) {materialFunction=/Engine/Functions/Engine_MaterialFunctions03/Blends/Blend_Darken}.Result
    Base Color=MaterialFunctionCall(Blend=@n14.Base Color, Base=@n15.RGB) {materialFunction=&Blend_Darken}.Result
   ) {attributeSetTypes [#1, #2]}
   False=@n13.Result
  ) {group=Opacity, parameterName=TuneEmissiveOutput}
  StaticSwitchParameter(
   True=SetMaterialAttributes!(MaterialAttributes=@n16, Opacity=Multiply(A=@n5, B=@n7)) {attributeSetTypes [#3]}
   False=@n16
  ) {group=Opacity, parameterName=Use UV2 Opacity Mask}
  VectorParameter() {group=z_code, parameterName=Code_DeformationVector, defaultValue=<rgba 0 0 1 0>}
 ]
}
